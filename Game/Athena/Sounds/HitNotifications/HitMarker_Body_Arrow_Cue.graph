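# Sound cue for the body-hit marker notification on arrow hits, routed through an instigator 2D hit-notification sound class and hit-marker attenuation.
# A Random node picks among five equally weighted arrow impact waves, which come from the "Target Critical" set despite the body-hit name.
# A shared hit-notification concurrency set limits overlap to two voices, stopping the oldest, with volume scaling implied by its name.
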SoundCue /Game/Athena/Sounds/HitNotifications/HitMarker_Body_Arrow_Cue
root=Random_3
nodes [
 {
  id=Random_3
  children [WavePlayer_14, WavePlayer_15, WavePlayer_16, WavePlayer_17, WavePlayer_18]
  properties {bShouldExcludeFromBranchCulling=true, weights [1,1,1,1,1]}
 }
 {id=WavePlayer_14, properties {soundWaveAssetPtr=/Game/Sounds/Fort_Weapons/Impacts/Impact_Arrow_Target_Critical_01}}
 {id=WavePlayer_15, properties {soundWaveAssetPtr=/Game/Sounds/Fort_Weapons/Impacts/Impact_Arrow_Target_Critical_02}}
 {id=WavePlayer_16, properties {soundWaveAssetPtr=/Game/Sounds/Fort_Weapons/Impacts/Impact_Arrow_Target_Critical_03}}
 {id=WavePlayer_17, properties {soundWaveAssetPtr=/Game/Sounds/Fort_Weapons/Impacts/Impact_Arrow_Target_Critical_04}}
 {id=WavePlayer_18, properties {soundWaveAssetPtr=/Game/Sounds/Fort_Weapons/Impacts/Impact_Arrow_Target_Critical_05}}
]
properties {
 attenuationSettings=&/Fort_HitMarker_Instigate_Att
 maxDistance=2097152
 soundClassObject=&Fort_2DHitNotifications_Instigate_PSM
 volumeMultiplier=0.631
 concurrencyOverrides {maxCount=2, resolutionRule=StopOldest}
 concurrencySet [&HitNotifications_TwoMaximum_StopOldest_VolumeScale]
}
