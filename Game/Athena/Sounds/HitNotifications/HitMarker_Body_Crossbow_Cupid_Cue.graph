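# Sound cue for the body-shot hit marker notification of the Cupid crossbow, played to the attacker who lands the hit.
# A single wave player feeds a modulator that slightly randomizes pitch on each play.
# It routes through the instigator 2D hit-notification sound class and attenuation, with concurrency capped at two voices that stops the oldest.
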SoundCue /Game/Athena/Sounds/HitNotifications/HitMarker_Body_Crossbow_Cupid_Cue
root=Modulator_0
nodes [
 {id=Modulator_0, children [WavePlayer_0], properties {pitchMax=1.1, pitchMin=0.9}}
 {id=WavePlayer_0, properties {soundWaveAssetPtr=&/sfx_crossbow_cupid_hit_normal}}
]
properties {
 attenuationSettings=&/Fort_HitMarker_Instigate_Att
 maxDistance=2097152
 soundClassObject=&Fort_2DHitNotifications_Instigate_PSM
 concurrencyOverrides {maxCount=2, resolutionRule=StopOldest}
 concurrencySet [&HitNotifications_TwoMaximum_StopOldest_VolumeScale]
}
