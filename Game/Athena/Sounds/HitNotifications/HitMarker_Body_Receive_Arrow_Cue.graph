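# Sound cue for the hit notification heard when a player receives an arrow hit to the body.
# A mixer layers a random, modulated arrow-on-player impact with a random UI tick that is pitch-varied and briefly delayed.
# Each layer has its own attenuation node, with the tick using the hit-marker receive attenuation.
# It plays in the received-hit sound class, with concurrency capping simultaneous notifications by stopping the oldest.
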SoundCue /Game/Athena/Sounds/HitNotifications/HitMarker_Body_Receive_Arrow_Cue
root=Mixer_0
nodes [
 {id=Attenuation_0, children [Delay_0], properties {attenuationSettings=&/Fort_HitMarker_Receive_Att}}
 {id=Attenuation_1, children [Modulator_0]}
 {id=Delay_0, children [Modulator_1], properties {delayMax=0.05, delayMin=0.05}}
 {id=Mixer_0, children [Attenuation_1, Attenuation_0], properties {inputVolume [1,1]}}
 {id=Modulator_0, children [Random_1]}
 {id=Modulator_1, children [Random_2], properties {pitchMax=1.1, volumeMax=1, volumeMin=1}}
 {
  id=Random_1
  children [WavePlayer_3,WavePlayer_4,WavePlayer_5,WavePlayer_6,WavePlayer_7,WavePlayer_8]
  properties {bShouldExcludeFromBranchCulling=true, weights [1,1,1,1,1,1]}
 }
 {id=Random_2, children [WavePlayer_12, WavePlayer_13, WavePlayer_14], properties {bShouldExcludeFromBranchCulling=true, weights [1,1,1]}}
 {id=WavePlayer_12, properties {soundWaveAssetPtr=&/HitMarker_Receive_UI_Tick_01}}
 {id=WavePlayer_13, properties {soundWaveAssetPtr=&/HitMarker_Receive_UI_Tick_02}}
 {id=WavePlayer_14, properties {soundWaveAssetPtr=&/HitMarker_Receive_UI_Tick_03}}
 {id=WavePlayer_3, properties {soundWaveAssetPtr=/Game/Sounds/Fort_Weapons/Impacts/Impact_Arrow_Player_Default_01}}
 {id=WavePlayer_4, properties {soundWaveAssetPtr=/Game/Sounds/Fort_Weapons/Impacts/Impact_Arrow_Player_Default_02}}
 {id=WavePlayer_5, properties {soundWaveAssetPtr=/Game/Sounds/Fort_Weapons/Impacts/Impact_Arrow_Player_Default_03}}
 {id=WavePlayer_6, properties {soundWaveAssetPtr=/Game/Sounds/Fort_Weapons/Impacts/Impact_Arrow_Player_Default_04}}
 {id=WavePlayer_7, properties {soundWaveAssetPtr=/Game/Sounds/Fort_Weapons/Impacts/Impact_Arrow_Player_Default_05}}
 {id=WavePlayer_8, properties {soundWaveAssetPtr=/Game/Sounds/Fort_Weapons/Impacts/Impact_Arrow_Player_Default_06}}
]
properties {
 attenuationSettings=&Fort_HitMarker_Receive_Att
 maxDistance=2097152
 soundClassObject=/Game/Sounds/FortSoundClassesAndModes/Fort_2DHitNotifications_Receive_PSM
 volumeMultiplier=0.631
 concurrencyOverrides {maxCount=2, resolutionRule=StopOldest}
 concurrencySet [&HitNotifications_TwoMaximum_StopOldest_VolumeScale]
}
