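# Sound cue for the hit-marker notification of a critical arrow hit, using an instigator hit-marker attenuation and 2D hit-notification sound class.
# A default arrow stinger plays immediately, mixed with a slightly quieter critical layer after a short fixed delay.
# That layer blends a critical stinger with one of five random critical impacts, each through its own Modulator.
# Concurrency limits it to two voices, stopping the oldest.
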SoundCue /Game/Athena/Sounds/HitNotifications/HitMarker_Critical_Arrow_Cue
root=Mixer_4
nodes [
 {id=Delay_0, children [Mixer_1], properties {delayMax=0.15, delayMin=0.15}}
 {id=Mixer_1, children [Modulator_1,Modulator_2], properties {inputVolume [0.8,0.5]}}
 {id=Mixer_4, children [Delay_0,WavePlayer_1], properties {inputVolume [0.8,1]}}
 {id=Modulator_1, children [WavePlayer_0]}
 {id=Modulator_2, children [Random_4]}
 {
  id=Random_4
  children [WavePlayer_13, WavePlayer_14, WavePlayer_15, WavePlayer_16, WavePlayer_17]
  properties {bShouldExcludeFromBranchCulling=true, weights [1,1,1,1,1]}
 }
 {id=WavePlayer_0, properties {soundWaveAssetPtr=/Game/Sounds/Fort_Weapons/Impacts/Impact_Arrow_Target_Critical_Stinger_01}}
 {id=WavePlayer_1, properties {soundWaveAssetPtr=/Game/Sounds/Fort_Weapons/Impacts/Impact_Arrow_Target_Default_Stinger_01}}
 {id=WavePlayer_13, properties {soundWaveAssetPtr=/Game/Sounds/Fort_Weapons/Impacts/Impact_Arrow_Target_Critical_01}}
 {id=WavePlayer_14, properties {soundWaveAssetPtr=/Game/Sounds/Fort_Weapons/Impacts/Impact_Arrow_Target_Critical_02}}
 {id=WavePlayer_15, properties {soundWaveAssetPtr=/Game/Sounds/Fort_Weapons/Impacts/Impact_Arrow_Target_Critical_03}}
 {id=WavePlayer_16, properties {soundWaveAssetPtr=/Game/Sounds/Fort_Weapons/Impacts/Impact_Arrow_Target_Critical_04}}
 {id=WavePlayer_17, properties {soundWaveAssetPtr=/Game/Sounds/Fort_Weapons/Impacts/Impact_Arrow_Target_Critical_05}}
]
properties {
 attenuationSettings=&/Fort_HitMarker_Instigate_Att
 maxDistance=2097152
 soundClassObject=&Fort_2DHitNotifications_Instigate_PSM
 volumeMultiplier=0.294
 concurrencyOverrides {maxCount=2, resolutionRule=StopOldest}
 concurrencySet [&HitNotifications_TwoMaximum_StopOldest_VolumeScale]
}
