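# Sound cue for the critical hit marker of the Cupid crossbow, played as a hit notification.
# A single wave player feeds the crossbow crit hit sample straight to the output.
# It uses instigator hit-marker attenuation and a 2D hit notification sound class.
# Concurrency allows two instances at once, stopping the oldest, through a shared hit-notification set that scales volume.
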SoundCue /Game/Athena/Sounds/HitNotifications/HitMarker_Critical_Crossbow_Cupid_Cue
root=WavePlayer_0
nodes [{id=WavePlayer_0, properties {soundWaveAssetPtr=&/sfx_crossbow_cupid_hit_crit}}]
properties {
 attenuationSettings=&/Fort_HitMarker_Instigate_Att
 maxDistance=2097152
 soundClassObject=&Fort_2DHitNotifications_Instigate_PSM
 concurrencyOverrides {maxCount=2, resolutionRule=StopOldest}
 concurrencySet [&HitNotifications_TwoMaximum_StopOldest_VolumeScale]
}
