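# Sound cue for the critical hit marker notification, layering a body hit sound with a delayed critical accent.
# The root mixer plays one of three random body waves through a modulator, alongside a second, fixed-pitch branch delayed briefly.
# That delayed branch mixes a quieter random flesh impact under a louder critical notification layer.
# Concurrency settings cap it at two instances, stopping the oldest, and it uses hit-marker attenuation and a 2D notification sound class.
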SoundCue /Game/Athena/Sounds/HitNotifications/HitMarker_Critical_Cue
root=Mixer_0
nodes [
 {id=Delay_6, children [Mixer_1], properties {delayMax=0.1, delayMin=0.1}}
 {id=Mixer_0, children [Modulator_4,Modulator_6], properties {inputVolume [1,0.8]}}
 {id=Mixer_1, children [Random_7,WavePlayer_1], properties {inputVolume [0.35,1.2]}}
 {id=Modulator_4, children [Random_2]}
 {id=Modulator_6, children [Delay_6], properties {pitchMax=1, pitchMin=1}}
 {id=Random_2, children [WavePlayer_4,WavePlayer_5,WavePlayer_8], properties {bShouldExcludeFromBranchCulling=true, weights [1,1,1]}}
 {
  id=Random_7
  children [WavePlayer_20, WavePlayer_21, WavePlayer_22, WavePlayer_23]
  properties {bShouldExcludeFromBranchCulling=true, weights [1,1,1,1]}
 }
 {id=WavePlayer_1, properties {soundWaveAssetPtr=&/sfx_hitmarker_notification_critical_layer}}
 {id=WavePlayer_20, properties {soundWaveAssetPtr=/Game/Sounds/Fort_Impact_Sounds/Flesh/player_impact_enemy_stereo_01}}
 {id=WavePlayer_21, properties {soundWaveAssetPtr=/Game/Sounds/Fort_Impact_Sounds/Flesh/player_impact_enemy_stereo_02}}
 {id=WavePlayer_22, properties {soundWaveAssetPtr=/Game/Sounds/Fort_Impact_Sounds/Flesh/player_impact_enemy_stereo_03}}
 {id=WavePlayer_23, properties {soundWaveAssetPtr=/Game/Sounds/Fort_Impact_Sounds/Flesh/player_impact_enemy_stereo_04}}
 {id=WavePlayer_4, properties {soundWaveAssetPtr=&/HitMarker_Body_01}}
 {id=WavePlayer_5, properties {soundWaveAssetPtr=&/HitMarker_Body_02}}
 {id=WavePlayer_8, properties {soundWaveAssetPtr=&/HitMarker_Body_03}}
]
properties {
 attenuationSettings=&/Fort_HitMarker_Instigate_Att
 maxDistance=2097152
 soundClassObject=&Fort_2DHitNotifications_Instigate_PSM
 volumeMultiplier=0.757
 concurrencyOverrides {maxCount=2, resolutionRule=StopOldest}
 concurrencySet [&HitNotifications_TwoMaximum_StopOldest_VolumeScale]
}
