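# Sound cue for the hit notification heard when receiving a critical arrow hit, routed to the 2D receive sound class.
# An impact branch mixes random player arrow impacts, random critical twangs and a slightly delayed thud.
# A parallel branch plays a randomized, pitch-modulated UI tick after a short delay through the hit-marker receive attenuation.
# Concurrency is set for two voices, stopping the oldest.
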SoundCue /Game/Athena/Sounds/HitNotifications/HitMarker_Critical_Receive_Arrow_Cue
root=Mixer_0
nodes [
 {id=Attenuation_1, children [Modulator_0]}
 {id=Attenuation_2, children [Delay_1], properties {attenuationSettings=&/Fort_HitMarker_Receive_Att}}
 {id=Delay_1, children [Modulator_2], properties {delayMax=0.05, delayMin=0.05}}
 {id=Delay_2, children [WavePlayer_42], properties {delayMax=0.1, delayMin=0.1}}
 {id=Mixer_0, children [Attenuation_1, Attenuation_2], properties {inputVolume [1,1]}}
 {id=Mixer_4, children [Modulator_5,Random_9,Delay_2], properties {inputVolume [0.5,0.6,0.6]}}
 {id=Modulator_0, children [Mixer_4]}
 {id=Modulator_2, children [Random_2], properties {pitchMax=1.1, volumeMax=1, volumeMin=1}}
 {id=Modulator_5, children [Random_7]}
 {id=Random_2, children [WavePlayer_1,WavePlayer_2,WavePlayer_3], properties {bShouldExcludeFromBranchCulling=true, weights [1,1,1]}}
 {
  id=Random_7
  children [WavePlayer_12, WavePlayer_13, WavePlayer_14, WavePlayer_15, WavePlayer_29, WavePlayer_30]
  properties {bShouldExcludeFromBranchCulling=true, weights [1,1,1,1,1,1]}
 }
 {
  id=Random_9
  children [WavePlayer_38, WavePlayer_39, WavePlayer_40, WavePlayer_41]
  properties {bShouldExcludeFromBranchCulling=true, weights [1,1,1,1]}
 }
 {id=WavePlayer_1, properties {soundWaveAssetPtr=&/HitMarker_Receive_UI_Tick_01}}
 {id=WavePlayer_12, properties {soundWaveAssetPtr=/Game/Sounds/Fort_Weapons/Impacts/Impact_Arrow_Player_Default_01}}
 {id=WavePlayer_13, properties {soundWaveAssetPtr=/Game/Sounds/Fort_Weapons/Impacts/Impact_Arrow_Player_Default_02}}
 {id=WavePlayer_14, properties {soundWaveAssetPtr=/Game/Sounds/Fort_Weapons/Impacts/Impact_Arrow_Player_Default_03}}
 {id=WavePlayer_15, properties {soundWaveAssetPtr=/Game/Sounds/Fort_Weapons/Impacts/Impact_Arrow_Player_Default_04}}
 {id=WavePlayer_2, properties {soundWaveAssetPtr=&/HitMarker_Receive_UI_Tick_02}}
 {id=WavePlayer_29, properties {soundWaveAssetPtr=/Game/Sounds/Fort_Weapons/Impacts/Impact_Arrow_Player_Default_05}}
 {id=WavePlayer_3, properties {soundWaveAssetPtr=&/HitMarker_Receive_UI_Tick_03}}
 {id=WavePlayer_30, properties {soundWaveAssetPtr=/Game/Sounds/Fort_Weapons/Impacts/Impact_Arrow_Player_Default_06}}
 {id=WavePlayer_38, properties {soundWaveAssetPtr=/Game/Sounds/Fort_Weapons/Impacts/Impact_Arrow_Player_Critical_Twang_01}}
 {id=WavePlayer_39, properties {soundWaveAssetPtr=/Game/Sounds/Fort_Weapons/Impacts/Impact_Arrow_Player_Critical_Twang_02}}
 {id=WavePlayer_40, properties {soundWaveAssetPtr=/Game/Sounds/Fort_Weapons/Impacts/Impact_Arrow_Player_Critical_Twang_03}}
 {id=WavePlayer_41, properties {soundWaveAssetPtr=/Game/Sounds/Fort_Weapons/Impacts/Impact_Arrow_Player_Critical_Twang_04}}
 {id=WavePlayer_42, properties {soundWaveAssetPtr=/Game/Sounds/Fort_Weapons/Impacts/Impact_Arrow_Target_Default_Thud_01}}
]
properties {
 attenuationSettings=&Fort_HitMarker_Receive_Att
 maxDistance=2097152
 soundClassObject=/Game/Sounds/FortSoundClassesAndModes/Fort_2DHitNotifications_Receive_PSM
 volumeMultiplier=0.841
 concurrencyOverrides {maxCount=2, resolutionRule=StopOldest}
 concurrencySet [&HitNotifications_TwoMaximum_StopOldest_VolumeScale]
}
