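# Sound cue for the hit notification heard when receiving a critical hit from the Cupid crossbow.
# A mixer layers the crossbow critical-hit wave with a briefly delayed, randomly chosen UI tick from three variants.
# Each layer has its own modulator and attenuation node, with the hit-marker receive attenuation set on the tick branch and cue.
# It uses the 2D receive hit-notification sound class and two-voice, stop-oldest concurrency.
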SoundCue /Game/Athena/Sounds/HitNotifications/HitMarker_Critical_Receive_Crossbow_Cupid_Cue
root=Mixer_0
nodes [
 {id=Attenuation_1, children [Modulator_1]}
 {id=Attenuation_2, children [Delay_1], properties {attenuationSettings=&/Fort_HitMarker_Receive_Att}}
 {id=Delay_1, children [Modulator_2], properties {delayMax=0.05, delayMin=0.05}}
 {id=Mixer_0, children [Attenuation_1, Attenuation_2], properties {inputVolume [1,1]}}
 {id=Modulator_1, children [WavePlayer_0]}
 {id=Modulator_2, children [Random_2], properties {pitchMax=1.1, volumeMax=1, volumeMin=1}}
 {id=Random_2, children [WavePlayer_12, WavePlayer_13, WavePlayer_14], properties {bShouldExcludeFromBranchCulling=true, weights [1,1,1]}}
 {id=WavePlayer_0, properties {soundWaveAssetPtr=&/sfx_crossbow_cupid_hit_crit}}
 {id=WavePlayer_12, properties {soundWaveAssetPtr=&/HitMarker_Receive_UI_Tick_01}}
 {id=WavePlayer_13, properties {soundWaveAssetPtr=&/HitMarker_Receive_UI_Tick_02}}
 {id=WavePlayer_14, properties {soundWaveAssetPtr=&/HitMarker_Receive_UI_Tick_03}}
]
properties {
 attenuationSettings=&Fort_HitMarker_Receive_Att
 maxDistance=2097152
 soundClassObject=/Game/Sounds/FortSoundClassesAndModes/Fort_2DHitNotifications_Receive_PSM
 volumeMultiplier=0.505
 concurrencyOverrides {maxCount=2, resolutionRule=StopOldest}
 concurrencySet [&HitNotifications_TwoMaximum_StopOldest_VolumeScale]
}
