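# Sound cue for the critical hit-received notification, routed to the 2D receive hit-notification sound class and capped at two instances, stopping the oldest.
# A quality node gives its first two levels a layered mix of enveloped bullet impacts plus delayed flesh and pitched-up metal impacts.
# The last two levels swap that stack for a single optimized wave.
# Both branches add a slightly delayed random UI tick with receive attenuation.
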SoundCue /Game/Athena/Sounds/HitNotifications/HitMarker_Critical_Receive_Cue
root=QualityLevel_0
nodes [
 {id=Attenuation_1, children [Modulator_6]}
 {id=Attenuation_3, children [Modulator_8]}
 {id=Attenuation_4, children [Delay_0], properties {attenuationSettings=&/Fort_HitMarker_Receive_Att}}
 {id=Attenuation_5, children [Delay_4], properties {attenuationSettings=&Fort_HitMarker_Receive_Att}}
 {id=Delay_0, children [Modulator_1], properties {delayMax=0.05, delayMin=0.05}}
 {id=Delay_1, children [Mixer_3], properties {delayMax=0.15, delayMin=0.15}}
 {id=Delay_4, children [Modulator_0], properties {delayMax=0.05, delayMin=0.05}}
 {
  id=Enveloper_1
  children [Random_4]
  properties {
   pitchCurve {editorCurveData {keys [{time=0}]}}
   volumeCurve {editorCurveData {keys [
      <at 0 1>
      {arriveTangent=32.5924, interpMode=RCIM_Cubic, leaveTangent=32.5924, tangentMode=RCTM_User, time=0.3, value=0}
      {arriveTangent=2.8166, interpMode=RCIM_Cubic, leaveTangent=2.8166, tangentMode=RCTM_User, time=0.4, value=1}
     ]}}
  }
 }
 {id=Mixer_2, children [Enveloper_1,Modulator_3], properties {inputVolume [1,1]}}
 {id=Mixer_3, children [Random_5,Modulator_4], properties {inputVolume [1,0.7]}}
 {id=Mixer_4, children [Attenuation_1, Attenuation_4], properties {inputVolume [1,1]}}
 {id=Mixer_5, children [Attenuation_3, Attenuation_5], properties {inputVolume [1,0.7]}}
 {id=Modulator_0, children [Random_0], properties {pitchMax=1.1, volumeMax=1, volumeMin=1}}
 {id=Modulator_1, children [Random_2], properties {pitchMax=1.1, volumeMax=1, volumeMin=1}}
 {id=Modulator_3, children [Delay_1], properties {pitchMax=1, pitchMin=1}}
 {id=Modulator_4, children [Random_6], properties {pitchMax=2, pitchMin=2}}
 {id=Modulator_6, children [Mixer_2]}
 {id=Modulator_8, children [WavePlayer_36]}
 {id=QualityLevel_0, children [Mixer_4,Mixer_4,Mixer_5,Mixer_5]}
 {id=Random_0, children [WavePlayer_0,WavePlayer_1,WavePlayer_4], properties {bShouldExcludeFromBranchCulling=true, weights [1,1,1]}}
 {id=Random_2, children [WavePlayer_12, WavePlayer_13, WavePlayer_14], properties {bShouldExcludeFromBranchCulling=true, weights [1,1,1]}}
 {
  id=Random_4
  children [WavePlayer_16, WavePlayer_17, WavePlayer_18, WavePlayer_19]
  properties {bShouldExcludeFromBranchCulling=true, weights [1,1,1,1]}
 }
 {
  id=Random_5
  children [WavePlayer_20, WavePlayer_21, WavePlayer_22, WavePlayer_23]
  properties {bShouldExcludeFromBranchCulling=true, weights [1,1,1,1]}
 }
 {
  id=Random_6
  children [WavePlayer_24, WavePlayer_25, WavePlayer_26, WavePlayer_27, WavePlayer_28]
  properties {bShouldExcludeFromBranchCulling=true, weights [1,1,1,1,1]}
 }
 {id=WavePlayer_0, properties {soundWaveAssetPtr=&/HitMarker_Receive_UI_Tick_01}}
 {id=WavePlayer_1, properties {soundWaveAssetPtr=&/HitMarker_Receive_UI_Tick_02}}
 {id=WavePlayer_12, properties {soundWaveAssetPtr=&HitMarker_Receive_UI_Tick_01}}
 {id=WavePlayer_13, properties {soundWaveAssetPtr=&HitMarker_Receive_UI_Tick_02}}
 {id=WavePlayer_14, properties {soundWaveAssetPtr=&/HitMarker_Receive_UI_Tick_03}}
 {id=WavePlayer_16, properties {soundWaveAssetPtr=/Game/Sounds/Fort_Weapons/Impacts/player_bulletimpact_rifle_01}}
 {id=WavePlayer_17, properties {soundWaveAssetPtr=/Game/Sounds/Fort_Weapons/Impacts/player_bulletimpact_rifle_02}}
 {id=WavePlayer_18, properties {soundWaveAssetPtr=/Game/Sounds/Fort_Weapons/Impacts/player_bulletimpact_rifle_03}}
 {id=WavePlayer_19, properties {soundWaveAssetPtr=/Game/Sounds/Fort_Weapons/Impacts/player_bulletimpact_rifle_04}}
 {id=WavePlayer_20, properties {soundWaveAssetPtr=/Game/Sounds/Fort_Impact_Sounds/Flesh/player_impact_enemy_stereo_01}}
 {id=WavePlayer_21, properties {soundWaveAssetPtr=/Game/Sounds/Fort_Impact_Sounds/Flesh/player_impact_enemy_stereo_02}}
 {id=WavePlayer_22, properties {soundWaveAssetPtr=/Game/Sounds/Fort_Impact_Sounds/Flesh/player_impact_enemy_stereo_03}}
 {id=WavePlayer_23, properties {soundWaveAssetPtr=/Game/Sounds/Fort_Impact_Sounds/Flesh/player_impact_enemy_stereo_04}}
 {id=WavePlayer_24, properties {soundWaveAssetPtr=/Game/Sounds/Fort_Impact_Sounds/Demolition/Metal/Metal_Cannon_Impact01}}
 {id=WavePlayer_25, properties {soundWaveAssetPtr=/Game/Sounds/Fort_Impact_Sounds/Demolition/Metal/Metal_Cannon_Impact02}}
 {id=WavePlayer_26, properties {soundWaveAssetPtr=/Game/Sounds/Fort_Impact_Sounds/Demolition/Metal/Metal_Cannon_Impact03}}
 {id=WavePlayer_27, properties {soundWaveAssetPtr=/Game/Sounds/Fort_Impact_Sounds/Demolition/Metal/Metal_Cannon_Impact04}}
 {id=WavePlayer_28, properties {soundWaveAssetPtr=/Game/Sounds/Fort_Impact_Sounds/Demolition/Metal/Metal_Cannon_Impact05}}
 {id=WavePlayer_36, properties {soundWaveAssetPtr=&/HitMarker_Critical_Receive_opt}}
 {id=WavePlayer_4, properties {soundWaveAssetPtr=&HitMarker_Receive_UI_Tick_03}}
]
properties {
 attenuationSettings=&Fort_HitMarker_Receive_Att
 maxDistance=2097152
 soundClassObject=/Game/Sounds/FortSoundClassesAndModes/Fort_2DHitNotifications_Receive_PSM
 volumeMultiplier=0.421
 concurrencyOverrides {maxCount=2, resolutionRule=StopOldest}
 concurrencySet [&HitNotifications_TwoMaximum_StopOldest_VolumeScale]
}
