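# Sound cue for the hit notification of a critical arrow kill, using the instigator 2D hit-notification class and hit-marker attenuation.
# A top mixer plays the default arrow stinger immediately, layered with two staggered delayed branches.
# The first delay feeds a critical stinger mixed with one of five random critical impacts, each through a modulator; the slightly later one adds the death stinger.
# A critical hit-notification concurrency set stops the oldest instance.
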SoundCue /Game/Athena/Sounds/HitNotifications/HitMarker_Death_Arrow_Critical_Cue
root=Mixer_4
nodes [
 {id=Delay_0, children [Mixer_0], properties {delayMax=0.15, delayMin=0.15}}
 {id=Delay_3, children [WavePlayer_5], properties {delayMax=0.18, delayMin=0.18}}
 {id=Mixer_0, children [Modulator_1,Modulator_2], properties {inputVolume [0.8,0.5]}}
 {id=Mixer_4, children [Delay_0,WavePlayer_8,Delay_3], properties {inputVolume [0.5,0.4,0.4]}}
 {id=Modulator_1, children [WavePlayer_7]}
 {id=Modulator_2, children [Random_0]}
 {
  id=Random_0
  children [WavePlayer_0,WavePlayer_1,WavePlayer_2,WavePlayer_4,WavePlayer_6]
  properties {bShouldExcludeFromBranchCulling=true, weights [1,1,1,1,1]}
 }
 {id=WavePlayer_0, properties {soundWaveAssetPtr=/Game/Sounds/Fort_Weapons/Impacts/Impact_Arrow_Target_Critical_01}}
 {id=WavePlayer_1, properties {soundWaveAssetPtr=/Game/Sounds/Fort_Weapons/Impacts/Impact_Arrow_Target_Critical_02}}
 {id=WavePlayer_2, properties {soundWaveAssetPtr=/Game/Sounds/Fort_Weapons/Impacts/Impact_Arrow_Target_Critical_03}}
 {id=WavePlayer_4, properties {soundWaveAssetPtr=/Game/Sounds/Fort_Weapons/Impacts/Impact_Arrow_Target_Critical_04}}
 {id=WavePlayer_5, properties {soundWaveAssetPtr=&/HitMarker_Death_Stinger_01}}
 {id=WavePlayer_6, properties {soundWaveAssetPtr=/Game/Sounds/Fort_Weapons/Impacts/Impact_Arrow_Target_Critical_05}}
 {id=WavePlayer_7, properties {soundWaveAssetPtr=/Game/Sounds/Fort_Weapons/Impacts/Impact_Arrow_Target_Critical_Stinger_01}}
 {id=WavePlayer_8, properties {soundWaveAssetPtr=/Game/Sounds/Fort_Weapons/Impacts/Impact_Arrow_Target_Default_Stinger_01}}
]
properties {
 attenuationSettings=&/Fort_HitMarker_Instigate_Att
 maxDistance=2097152
 soundClassObject=&Fort_2DHitNotifications_Instigate_PSM
 volumeMultiplier=0.794
 concurrencyOverrides {maxCount=2, resolutionRule=StopOldest}
 concurrencySet [/Game/Sounds/Concurrency_Rules/HitNotifications_Critical_StopOldest]
}
