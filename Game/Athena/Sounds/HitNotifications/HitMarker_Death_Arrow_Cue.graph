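# Arrow death hit-marker sound cue, classed as a 2D instigator hit notification.
# A mixer layers one of five random default arrow target impacts with two quieter layers on matching short delays: an arrow death impact and a pitched-down death stinger.
# The random node is excluded from branch culling, and a critical hit-notification concurrency set governs overlapping plays by stopping the oldest.
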SoundCue /Game/Athena/Sounds/HitNotifications/HitMarker_Death_Arrow_Cue
root=Mixer_0
nodes [
 {id=Delay_0, children [Modulator_3], properties {delayMax=0.15, delayMin=0.15}}
 {id=Delay_1, children [WavePlayer_16], properties {delayMax=0.15, delayMin=0.15}}
 {id=Mixer_0, children [Delay_0,Delay_1,Random_0], properties {inputVolume [0.4,0.6,0.8]}}
 {id=Modulator_3, children [WavePlayer_5], properties {pitchMax=0.9, pitchMin=0.9, volumeMax=1, volumeMin=1}}
 {
  id=Random_0
  children [WavePlayer_0,WavePlayer_1,WavePlayer_2,WavePlayer_3,WavePlayer_4]
  properties {bShouldExcludeFromBranchCulling=true, weights [1,1,1,1,1]}
 }
 {id=WavePlayer_0, properties {soundWaveAssetPtr=/Game/Sounds/Fort_Weapons/Impacts/Impact_Arrow_Target_Default_01}}
 {id=WavePlayer_1, properties {soundWaveAssetPtr=/Game/Sounds/Fort_Weapons/Impacts/Impact_Arrow_Target_Default_02}}
 {id=WavePlayer_16, properties {soundWaveAssetPtr=/Game/Sounds/Fort_Weapons/Impacts/Impact_Arrow_Target_Death_01}}
 {id=WavePlayer_2, properties {soundWaveAssetPtr=/Game/Sounds/Fort_Weapons/Impacts/Impact_Arrow_Target_Default_03}}
 {id=WavePlayer_3, properties {soundWaveAssetPtr=/Game/Sounds/Fort_Weapons/Impacts/Impact_Arrow_Target_Default_04}}
 {id=WavePlayer_4, properties {soundWaveAssetPtr=/Game/Sounds/Fort_Weapons/Impacts/Impact_Arrow_Target_Default_05}}
 {id=WavePlayer_5, properties {soundWaveAssetPtr=&/HitMarker_Death_Stinger_01}}
]
properties {
 attenuationSettings=&/Fort_HitMarker_Instigate_Att
 maxDistance=2097152
 soundClassObject=&Fort_2DHitNotifications_Instigate_PSM
 volumeMultiplier=0.688
 concurrencyOverrides {maxCount=2, resolutionRule=StopOldest}
 concurrencySet [/Game/Sounds/Concurrency_Rules/HitNotifications_Critical_StopOldest]
}
