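# Sound cue for the critical death hit marker, layering an elimination hit with a death stinger.
# A mixer combines a randomly picked, modulated Elim wave at full volume with a much quieter stinger, each on a short fixed delay, the stinger landing slightly later.
# It uses hit-marker instigator attenuation, a 2D hit-notification sound class and a critical-notification concurrency rule that stops the oldest sound.
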SoundCue /Game/Athena/Sounds/HitNotifications/HitMarker_Death_Critical_Cue
root=Mixer_0
nodes [
 {id=Delay_1, children [WavePlayer_1], properties {delayMax=0.15, delayMin=0.15}}
 {id=Delay_2, children [Modulator_1], properties {delayMax=0.1, delayMin=0.1}}
 {id=Mixer_0, children [Delay_2,Delay_1], properties {inputVolume [1,0.2]}}
 {id=Modulator_1, children [Random_0]}
 {id=Random_0, children [WavePlayer_2,WavePlayer_3], properties {bShouldExcludeFromBranchCulling=true, weights [1,1]}}
 {id=WavePlayer_1, properties {soundWaveAssetPtr=&/HitMarker_Death_Stinger_01}}
 {id=WavePlayer_2, properties {soundWaveAssetPtr=&/HitMarker_Elim_01}}
 {id=WavePlayer_3, properties {soundWaveAssetPtr=&/HitMarker_Elim_02}}
]
properties {
 attenuationSettings=&/Fort_HitMarker_Instigate_Att
 maxDistance=2097152
 soundClassObject=&Fort_2DHitNotifications_Instigate_PSM
 concurrencyOverrides {maxCount=2, resolutionRule=StopOldest}
 concurrencySet [/Game/Sounds/Concurrency_Rules/HitNotifications_Critical_StopOldest]
}
