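# Hit notification sound cue for receiving a critical death hit, named as the arrow variant but playing the generic critical-death receive waves.
# A quality-level switch sends its first two slots to an equal-weight random pick of four variations and its last two to a single optimized wave.
# It uses the received-hit sound class and attenuation, with stop-oldest concurrency limiting overlapping plays.
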SoundCue /Game/Athena/Sounds/HitNotifications/HitMarker_Death_Critical_Receive_Arrow_Cue
root=QualityLevel_1
nodes [
 {id=QualityLevel_1, children [Random_3,Random_3,WavePlayer_1,WavePlayer_1]}
 {
  id=Random_3
  children [WavePlayer_3, WavePlayer_17, WavePlayer_18, WavePlayer_19]
  properties {bShouldExcludeFromBranchCulling=true, weights [1,1,1,1]}
 }
 {id=WavePlayer_1, properties {soundWaveAssetPtr=&/HitMarker_Death_Critical_Receive_opt}}
 {id=WavePlayer_17, properties {soundWaveAssetPtr=&/HitMarker_Death_Critical_Receive_02}}
 {id=WavePlayer_18, properties {soundWaveAssetPtr=&/HitMarker_Death_Critical_Receive_03}}
 {id=WavePlayer_19, properties {soundWaveAssetPtr=&/HitMarker_Death_Critical_Receive_04}}
 {id=WavePlayer_3, properties {soundWaveAssetPtr=&/HitMarker_Death_Critical_Receive_01}}
]
properties {
 attenuationSettings=&/Fort_HitMarker_Receive_Att
 maxDistance=2097152
 soundClassObject=/Game/Sounds/FortSoundClassesAndModes/Fort_2DHitNotifications_Receive_PSM
 volumeMultiplier=0.494
 concurrencyOverrides {maxCount=2, resolutionRule=StopOldest}
 concurrencySet [/Game/Sounds/Concurrency_Rules/HitNotifications_Critical_StopOldest]
}
