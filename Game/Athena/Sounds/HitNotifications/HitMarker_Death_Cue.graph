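# Sound cue for the death hit-marker notification, layering an elimination sound with a death stinger.
# A mixer plays a briefly delayed, modulated random pick of two elimination waves at full level over a quiet, pitched-down stinger.
# It routes through the 2D instigator hit-notification sound class with hit-marker attenuation.
# A critical hit-notification concurrency rule is assigned, alongside stop-oldest override settings.
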SoundCue /Game/Athena/Sounds/HitNotifications/HitMarker_Death_Cue
root=Mixer_0
nodes [
 {id=Delay_0, children [Modulator_0], properties {delayMax=0.1, delayMin=0.1}}
 {id=Mixer_0, children [Delay_0,Modulator_1], properties {inputVolume [1,0.15]}}
 {id=Modulator_0, children [Random_0]}
 {id=Modulator_1, children [WavePlayer_2], properties {pitchMax=0.9, pitchMin=0.9, volumeMax=1, volumeMin=1}}
 {id=Random_0, children [WavePlayer_0,WavePlayer_1], properties {bShouldExcludeFromBranchCulling=true, weights [1,1]}}
 {id=WavePlayer_0, properties {soundWaveAssetPtr=&/HitMarker_Elim_01}}
 {id=WavePlayer_1, properties {soundWaveAssetPtr=&/HitMarker_Elim_02}}
 {id=WavePlayer_2, properties {soundWaveAssetPtr=&/HitMarker_Death_Stinger_01}}
]
properties {
 attenuationSettings=&/Fort_HitMarker_Instigate_Att
 maxDistance=2097152
 soundClassObject=&Fort_2DHitNotifications_Instigate_PSM
 concurrencyOverrides {maxCount=2, resolutionRule=StopOldest}
 concurrencySet [/Game/Sounds/Concurrency_Rules/HitNotifications_Critical_StopOldest]
}
