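# Sound cue for the received hit-marker notification on a fall death, judging by its name, routed to the 2D hit-notification receive sound class.
# A mixer layers an immediate random bullet impact under a fade-out envelope, a delayed, louder random bone crunch, and a later big-hit marker.
# A quality node uses that mix for its first two levels and a single optimized wave for the other two.
# Concurrency caps it at two instances, stopping the oldest.
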SoundCue /Game/Athena/Sounds/HitNotifications/HitMarker_Death_Fall_Receive_Cue
root=QualityLevel_0
nodes [
 {id=Delay_0, children [Random_1], properties {delayMax=0.25, delayMin=0.25}}
 {id=Delay_1, children [WavePlayer_5], properties {delayMax=0.5, delayMin=0.5}}
 {
  id=Enveloper_0
  children [Random_0]
  properties {pitchCurve {editorCurveData {keys [{time=0}]}}, volumeCurve {editorCurveData {keys [<at 0 1>,<at 0.75 0>]}}}
 }
 {id=Mixer_0, children [Delay_1,Enveloper_0,Delay_0], properties {inputVolume [1,1,2]}}
 {id=QualityLevel_0, children [Mixer_0,Mixer_0,WavePlayer_8,WavePlayer_8]}
 {
  id=Random_0
  children [WavePlayer_0,WavePlayer_1,WavePlayer_3,WavePlayer_4]
  properties {bShouldExcludeFromBranchCulling=true, weights [1,1,1,1]}
 }
 {id=Random_1, children [WavePlayer_2,WavePlayer_6], properties {bShouldExcludeFromBranchCulling=true, weights [1,1]}}
 {id=WavePlayer_0, properties {soundWaveAssetPtr=/Game/Sounds/Fort_Weapons/Impacts/player_bulletimpact_rifle_01}}
 {id=WavePlayer_1, properties {soundWaveAssetPtr=/Game/Sounds/Fort_Weapons/Impacts/player_bulletimpact_rifle_02}}
 {id=WavePlayer_2, properties {soundWaveAssetPtr=/Game/Athena/Sounds/Character/Falling/sfx_Falling_BoneCrunch_01}}
 {id=WavePlayer_3, properties {soundWaveAssetPtr=/Game/Sounds/Fort_Weapons/Impacts/player_bulletimpact_rifle_03}}
 {id=WavePlayer_4, properties {soundWaveAssetPtr=/Game/Sounds/Fort_Weapons/Impacts/player_bulletimpact_rifle_04}}
 {id=WavePlayer_5, properties {soundWaveAssetPtr=&/sfx_hitmarker_bighit}}
 {id=WavePlayer_6, properties {soundWaveAssetPtr=/Game/Athena/Sounds/Character/Falling/sfx_Falling_BoneCrunch_02}}
 {id=WavePlayer_8, properties {soundWaveAssetPtr=&/HitMarker_Death_Fall_Receive_opt}}
]
properties {
 attenuationSettings=&/Fort_HitMarker_Receive_Att
 maxDistance=2097152
 soundClassObject=/Game/Sounds/FortSoundClassesAndModes/Fort_2DHitNotifications_Receive_PSM
 volumeMultiplier=0.494
 concurrencyOverrides {maxCount=2, resolutionRule=StopOldest}
 concurrencySet [&HitNotifications_TwoMaximum_StopOldest_VolumeScale]
}
