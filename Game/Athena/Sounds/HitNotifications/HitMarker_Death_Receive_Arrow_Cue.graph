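# Sound cue for the hit notification a player hears when killed by an arrow, as its name indicates.
# A mixer layers one of six equally weighted, randomly chosen arrow-on-player impacts with a quieter big-hit marker sting delayed by a fixed fraction of a second.
# It routes through the received-hit 2D notification sound class with hitmarker attenuation and a concurrency set that caps overlapping hit notifications.
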SoundCue /Game/Athena/Sounds/HitNotifications/HitMarker_Death_Receive_Arrow_Cue
root=Mixer_1
nodes [
 {id=Delay_0, children [WavePlayer_10], properties {delayMax=0.2, delayMin=0.2}}
 {id=Mixer_1, children [Delay_0,Random_3], properties {inputVolume [0.35,0.6]}}
 {
  id=Random_3
  children [WavePlayer_16, WavePlayer_17, WavePlayer_18, WavePlayer_19, WavePlayer_20, WavePlayer_21]
  properties {bShouldExcludeFromBranchCulling=true, weights [1,1,1,1,1,1]}
 }
 {id=WavePlayer_10, properties {soundWaveAssetPtr=&/sfx_hitmarker_bighit}}
 {id=WavePlayer_16, properties {soundWaveAssetPtr=/Game/Sounds/Fort_Weapons/Impacts/Impact_Arrow_Player_Default_01}}
 {id=WavePlayer_17, properties {soundWaveAssetPtr=/Game/Sounds/Fort_Weapons/Impacts/Impact_Arrow_Player_Default_02}}
 {id=WavePlayer_18, properties {soundWaveAssetPtr=/Game/Sounds/Fort_Weapons/Impacts/Impact_Arrow_Player_Default_03}}
 {id=WavePlayer_19, properties {soundWaveAssetPtr=/Game/Sounds/Fort_Weapons/Impacts/Impact_Arrow_Player_Default_04}}
 {id=WavePlayer_20, properties {soundWaveAssetPtr=/Game/Sounds/Fort_Weapons/Impacts/Impact_Arrow_Player_Default_05}}
 {id=WavePlayer_21, properties {soundWaveAssetPtr=/Game/Sounds/Fort_Weapons/Impacts/Impact_Arrow_Player_Default_06}}
]
properties {
 attenuationSettings=&/Fort_HitMarker_Receive_Att
 maxDistance=2097152
 soundClassObject=/Game/Sounds/FortSoundClassesAndModes/Fort_2DHitNotifications_Receive_PSM
 volumeMultiplier=1.413
 concurrencyOverrides {maxCount=2, resolutionRule=StopOldest}
 concurrencySet [&HitNotifications_TwoMaximum_StopOldest_VolumeScale]
}
