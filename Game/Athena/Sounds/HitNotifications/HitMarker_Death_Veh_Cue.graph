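# Sound cue for the vehicle death hit marker stinger in Athena's hit notifications.
# A single wave player plays the Hitmarker_Death_Stinger_Vehicle wave with no randomization or mixing.
# It routes through the 2D instigator hit-notification sound class and instigator attenuation.
# A shared critical hit-notification concurrency set is assigned, alongside override settings that cap simultaneous plays and stop the oldest.
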SoundCue /Game/Athena/Sounds/HitNotifications/HitMarker_Death_Veh_Cue
root=WavePlayer_3
nodes [{id=WavePlayer_3, properties {soundWaveAssetPtr=&/Hitmarker_Death_Stinger_Vehicle}}]
properties {
 attenuationSettings=&/Fort_HitMarker_Instigate_Att
 maxDistance=2097152
 soundClassObject=&Fort_2DHitNotifications_Instigate_PSM
 volumeMultiplier=1
 concurrencyOverrides {maxCount=2, resolutionRule=StopOldest}
 concurrencySet [/Game/Sounds/Concurrency_Rules/HitNotifications_Critical_StopOldest]
}
