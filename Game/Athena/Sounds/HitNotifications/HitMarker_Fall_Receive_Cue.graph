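# Sound cue for the fall hit-marker notification on the receiving side, routed through the 2D receive hit-notification sound class.
# A quality switch picks the mix: the top two tiers layer a louder random bone crunch with a quieter, briefly delayed, fading rifle-impact sound, the second using fewer impact variants.
# The lower two tiers share one modulated, optimized wave.
# Concurrency limits stop the oldest instance beyond two.
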SoundCue /Game/Athena/Sounds/HitNotifications/HitMarker_Fall_Receive_Cue
root=QualityLevel_0
nodes [
 {id=Delay_0, children [Enveloper_0], properties {delayMax=0.05, delayMin=0.05}}
 {id=Delay_2, children [Enveloper_1], properties {delayMax=0.05, delayMin=0.05}}
 {
  id=Enveloper_0
  children [Random_2]
  properties {pitchCurve {editorCurveData {keys [{time=0}]}}, volumeCurve {editorCurveData {keys [<at 0 1>,<at 0.5 0>]}}}
 }
 {
  id=Enveloper_1
  children [Random_1]
  properties {pitchCurve {editorCurveData {keys [{time=0}]}}, volumeCurve {editorCurveData {keys [<at 0 1>,<at 0.5 0>]}}}
 }
 {id=Mixer_0, children [Delay_2,Random_0], properties {inputVolume [0.5,2]}}
 {id=Mixer_1, children [Delay_0,Random_3], properties {inputVolume [0.5,2]}}
 {id=Modulator_1, children [WavePlayer_8]}
 {id=QualityLevel_0, children [Mixer_0,Mixer_1,Modulator_1,Modulator_1]}
 {id=Random_0, children [WavePlayer_0,WavePlayer_1], properties {bShouldExcludeFromBranchCulling=true, weights [1,1]}}
 {
  id=Random_1
  children [WavePlayer_4,WavePlayer_5,WavePlayer_6,WavePlayer_7]
  properties {bShouldExcludeFromBranchCulling=true, weights [1,1,1,1]}
 }
 {id=Random_2, children [WavePlayer_9, WavePlayer_10], properties {bShouldExcludeFromBranchCulling=true, weights [1,1]}}
 {id=Random_3, children [WavePlayer_13, WavePlayer_14], properties {bShouldExcludeFromBranchCulling=true, weights [1,1]}}
 {id=WavePlayer_0, properties {soundWaveAssetPtr=/Game/Athena/Sounds/Character/Falling/sfx_Falling_BoneCrunch_06}}
 {id=WavePlayer_1, properties {soundWaveAssetPtr=/Game/Athena/Sounds/Character/Falling/sfx_Falling_BoneCrunch_05}}
 {id=WavePlayer_10, properties {soundWaveAssetPtr=/Game/Sounds/Fort_Weapons/Impacts/player_bulletimpact_rifle_02}}
 {id=WavePlayer_13, properties {soundWaveAssetPtr=&sfx_Falling_BoneCrunch_06}}
 {id=WavePlayer_14, properties {soundWaveAssetPtr=&sfx_Falling_BoneCrunch_05}}
 {id=WavePlayer_4, properties {soundWaveAssetPtr=/Game/Sounds/Fort_Weapons/Impacts/player_bulletimpact_rifle_01}}
 {id=WavePlayer_5, properties {soundWaveAssetPtr=&player_bulletimpact_rifle_02}}
 {id=WavePlayer_6, properties {soundWaveAssetPtr=/Game/Sounds/Fort_Weapons/Impacts/player_bulletimpact_rifle_03}}
 {id=WavePlayer_7, properties {soundWaveAssetPtr=/Game/Sounds/Fort_Weapons/Impacts/player_bulletimpact_rifle_04}}
 {id=WavePlayer_8, properties {soundWaveAssetPtr=&/HitMarker_Fall_Receive_opt}}
 {id=WavePlayer_9, properties {soundWaveAssetPtr=&player_bulletimpact_rifle_01}}
]
properties {
 attenuationSettings=&/Fort_HitMarker_Receive_Att
 maxDistance=2097152
 soundClassObject=/Game/Sounds/FortSoundClassesAndModes/Fort_2DHitNotifications_Receive_PSM
 volumeMultiplier=0.35
 concurrencyOverrides {maxCount=2, resolutionRule=StopOldest}
 concurrencySet [&HitNotifications_TwoMaximum_StopOldest_VolumeScale]
}
