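# Sound cue for the hit marker played on shield hits with the Cupid crossbow, routed through a 2D instigator hit-notification sound class.
# A mixer layers a pitch-varied crossbow hit sound, weighted louder, over one of four equally weighted third-person shield hit variants picked at random.
# Instigator attenuation shapes playback, and concurrency limits it to two instances, stopping the oldest.
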SoundCue /Game/Athena/Sounds/HitNotifications/HitMarker_Shield_Cupid_Cue
root=Mixer_0
nodes [
 {id=Mixer_0, children [Modulator_3,Modulator_4], properties {inputVolume [1.5,0.8]}}
 {id=Modulator_3, children [WavePlayer_19], properties {pitchMax=1.1, pitchMin=0.9}}
 {id=Modulator_4, children [Random_4], properties {volumeMax=1, volumeMin=1}}
 {
  id=Random_4
  children [WavePlayer_20, WavePlayer_21, WavePlayer_22, WavePlayer_23]
  properties {bShouldExcludeFromBranchCulling=true, weights [1,1,1,1]}
 }
 {id=WavePlayer_19, properties {soundWaveAssetPtr=&/sfx_crossbow_cupid_hit_normal}}
 {id=WavePlayer_20, properties {soundWaveAssetPtr=&/HitMarker_Shield_3P_01}}
 {id=WavePlayer_21, properties {soundWaveAssetPtr=&/HitMarker_Shield_3P_02}}
 {id=WavePlayer_22, properties {soundWaveAssetPtr=&/HitMarker_Shield_3P_03}}
 {id=WavePlayer_23, properties {soundWaveAssetPtr=&/HitMarker_Shield_3P_04}}
]
properties {
 attenuationSettings=&/Fort_HitMarker_Instigate_Att
 maxDistance=2097152
 soundClassObject=&Fort_2DHitNotifications_Instigate_PSM
 volumeMultiplier=0.547
 concurrencyOverrides {maxCount=2, resolutionRule=StopOldest}
 concurrencySet [&HitNotifications_TwoMaximum_StopOldest_VolumeScale]
}
