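# Hit notification Sound Cue for a critical hit that destroys an enemy's shield, using instigator hit-marker attenuation and sound class.
# A root mixer plays a briefly delayed random shield-break variant alongside a quality-switched branch.
# That branch, wired only for the first two quality slots, follows slightly later with a random flesh impact mixed with a louder critical-notification layer.
# A shared critical-notification concurrency set limits stacking by stopping the oldest.
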SoundCue /Game/Athena/Sounds/HitNotifications/HitMarker_Shield_Destroyed_Crit_Cue
root=Mixer_2
nodes [
 {id=Delay_0, children [Random_0], properties {delayMax=0.15, delayMin=0.15}}
 {id=Delay_2, children [Mixer_3], properties {delayMax=0.2, delayMin=0.2}}
 {id=Mixer_2, children [Delay_0, QualityLevel_1], properties {inputVolume [1,1]}}
 {id=Mixer_3, children [Random_2, WavePlayer_10], properties {inputVolume [1,1.25]}}
 {id=Mixer_4, children [null,Modulator_0], properties {inputVolume [1,1]}}
 {id=Modulator_0, children [Delay_2], properties {pitchMax=1, pitchMin=1}}
 {id=QualityLevel_1, children [Mixer_4,Mixer_4,null,null]}
 {id=Random_0, children [WavePlayer_0,WavePlayer_1], properties {bShouldExcludeFromBranchCulling=true, weights [1,1]}}
 {
  id=Random_2
  children [WavePlayer_6,WavePlayer_7,WavePlayer_8,WavePlayer_9]
  properties {bShouldExcludeFromBranchCulling=true, weights [1,1,1,1]}
 }
 {id=WavePlayer_0, properties {soundWaveAssetPtr=&/HitMarker_ShieldBreak_01}}
 {id=WavePlayer_1, properties {soundWaveAssetPtr=&/HitMarker_ShieldBreak_02}}
 {id=WavePlayer_10, properties {soundWaveAssetPtr=&/sfx_hitmarker_notification_critical_layer}}
 {id=WavePlayer_6, properties {soundWaveAssetPtr=/Game/Sounds/Fort_Impact_Sounds/Flesh/player_impact_enemy_stereo_01}}
 {id=WavePlayer_7, properties {soundWaveAssetPtr=/Game/Sounds/Fort_Impact_Sounds/Flesh/player_impact_enemy_stereo_02}}
 {id=WavePlayer_8, properties {soundWaveAssetPtr=/Game/Sounds/Fort_Impact_Sounds/Flesh/player_impact_enemy_stereo_03}}
 {id=WavePlayer_9, properties {soundWaveAssetPtr=/Game/Sounds/Fort_Impact_Sounds/Flesh/player_impact_enemy_stereo_04}}
]
properties {
 attenuationSettings=&/Fort_HitMarker_Instigate_Att
 maxDistance=2097152
 soundClassObject=&Fort_2DHitNotifications_Instigate_PSM
 volumeMultiplier=0.6
 concurrencyOverrides {maxCount=2, resolutionRule=StopOldest}
 concurrencySet [/Game/Sounds/Concurrency_Rules/HitNotifications_Critical_StopOldest]
}
