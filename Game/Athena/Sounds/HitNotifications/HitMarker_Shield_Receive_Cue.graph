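# Sound cue for the shield hit-marker notification played when receiving a hit, routed to the 2D receive-shield hit notification sound class.
# A Random node picks one of four equally weighted first-person shield waves and feeds a Modulator that varies pitch while holding volume steady.
# The cue plays at reduced volume with a receive-specific attenuation asset.
# A shared concurrency set, named for a two-voice stop-oldest limit with volume scaling, governs overlapping hits.
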SoundCue /Game/Athena/Sounds/HitNotifications/HitMarker_Shield_Receive_Cue
root=Modulator_0
nodes [
 {id=Modulator_0, children [Random_0], properties {pitchMax=1.1, volumeMax=1, volumeMin=1}}
 {
  id=Random_0
  children [WavePlayer_0,WavePlayer_1,WavePlayer_2,WavePlayer_3]
  properties {bShouldExcludeFromBranchCulling=true, weights [1,1,1,1]}
 }
 {id=WavePlayer_0, properties {soundWaveAssetPtr=&/HitMarker_Shield_1P_01}}
 {id=WavePlayer_1, properties {soundWaveAssetPtr=&/HitMarker_Shield_1P_02}}
 {id=WavePlayer_2, properties {soundWaveAssetPtr=&/HitMarker_Shield_1P_03}}
 {id=WavePlayer_3, properties {soundWaveAssetPtr=&/HitMarker_Shield_1P_04}}
]
properties {
 attenuationSettings=&/Fort_HitMarker_Receive_Att
 maxDistance=2097152
 soundClassObject=/Game/Sounds/FortSoundClassesAndModes/Fort_2DHitNotifications_Receive_Shield
 volumeMultiplier=0.6
 concurrencyOverrides {maxCount=2, resolutionRule=StopOldest}
 concurrencySet [&HitNotifications_TwoMaximum_StopOldest_VolumeScale]
}
