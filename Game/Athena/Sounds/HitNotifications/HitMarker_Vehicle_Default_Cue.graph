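# Default vehicle hit-marker sound cue, judging by its name, routed through a 2D instigator hit-notification sound class.
# A Random node, excluded from branch culling, picks equally among three light metal impact waves.
# A root Modulator varies pitch and volume downward on each play, under a reduced overall volume.
# It references shared hit-marker attenuation and a concurrency set named for two maximum voices, stop-oldest and volume scaling.
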SoundCue /Game/Athena/Sounds/HitNotifications/HitMarker_Vehicle_Default_Cue
root=Modulator_0
nodes [
 {id=Modulator_0, children [Random_0], properties {pitchMax=1, pitchMin=0.7, volumeMax=1, volumeMin=0.6}}
 {id=Random_0, children [WavePlayer_4, WavePlayer_9, WavePlayer_10], properties {bShouldExcludeFromBranchCulling=true, weights [1,1,1]}}
 {id=WavePlayer_10, properties {soundWaveAssetPtr=/Game/Sounds/Fort_Weapons/Impacts/MTL_IMPACT_LIGHT_03}}
 {id=WavePlayer_4, properties {soundWaveAssetPtr=/Game/Sounds/Fort_Weapons/Impacts/MTL_IMPACT_LIGHT_01}}
 {id=WavePlayer_9, properties {soundWaveAssetPtr=/Game/Sounds/Fort_Weapons/Impacts/MTL_IMPACT_LIGHT_02}}
]
properties {
 attenuationSettings=&/Fort_HitMarker_Instigate_Att
 maxDistance=2097152
 soundClassObject=&Fort_2DHitNotifications_Instigate_PSM
 volumeMultiplier=0.4
 attenuationOverrides {bAttenuate=false, bEnableListenerFocus=true, nonFocusVolumeAttenuation=0.1}
 concurrencyOverrides {maxCount=2, resolutionRule=StopOldest}
 concurrencySet [&HitNotifications_TwoMaximum_StopOldest_VolumeScale]
}
