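# Sound cue for the hit-marker notification tied to the Hamster vehicle, played from its impact damage samples.
# A quality-level switch picks a random node with four sample variations for the first level, or one with two for the other three.
# A modulator randomly lowers pitch and volume on each play.
# It uses instigator hit-marker attenuation and a 2D hit-notification class, with concurrency capped at two, stopping oldest.
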SoundCue /Game/Athena/Sounds/HitNotifications/HitMarker_Vehicle_Hamster_Cue
root=Modulator_0
nodes [
 {id=Modulator_0, children [QualityLevel_0], properties {pitchMax=1, pitchMin=0.7, volumeMax=1, volumeMin=0.6}}
 {id=QualityLevel_0, children [Random_2,Random_0,Random_0,Random_0]}
 {id=Random_0, children [WavePlayer_0,WavePlayer_1], properties {bShouldExcludeFromBranchCulling=true, weights [1,1]}}
 {
  id=Random_2
  children [WavePlayer_3,WavePlayer_5,WavePlayer_6,WavePlayer_7]
  properties {bShouldExcludeFromBranchCulling=true, weights [1,1,1,1]}
 }
 {id=WavePlayer_0, properties {soundWaveAssetPtr=/Game/Athena/Sounds/Vehicles/Hamster/Hamster_Impact_Damage_01}}
 {id=WavePlayer_1, properties {soundWaveAssetPtr=/Game/Athena/Sounds/Vehicles/Hamster/Hamster_Impact_Damage_02}}
 {id=WavePlayer_3, properties {soundWaveAssetPtr=&Hamster_Impact_Damage_01}}
 {id=WavePlayer_5, properties {soundWaveAssetPtr=&Hamster_Impact_Damage_02}}
 {id=WavePlayer_6, properties {soundWaveAssetPtr=/Game/Athena/Sounds/Vehicles/Hamster/Hamster_Impact_Damage_03}}
 {id=WavePlayer_7, properties {soundWaveAssetPtr=/Game/Athena/Sounds/Vehicles/Hamster/Hamster_Impact_Damage_04}}
]
properties {
 attenuationSettings=&/Fort_HitMarker_Instigate_Att
 maxDistance=2097152
 soundClassObject=&Fort_2DHitNotifications_Instigate_PSM
 volumeMultiplier=0.8
 concurrencyOverrides {maxCount=2, resolutionRule=StopOldest}
 concurrencySet [&HitNotifications_TwoMaximum_StopOldest_VolumeScale]
}
