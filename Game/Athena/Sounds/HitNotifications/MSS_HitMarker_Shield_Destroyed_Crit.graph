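# One-shot stereo MetaSound for the hit marker notification when a critical hit destroys a shield.
# Three MSS_Shield_Deactivate_Base instances layer synth shield-break, critical crash and enemy impact waves, sharing retrigger threshold and gain settings.
# Synth starts after a fixed delay, while crash and impact share a delay chosen by comparing time since last play to the threshold.
# Layers are mixed and attenuated, and any layer finishing signals OnFinished.
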MetaSoundSource /Game/Athena/Sounds/HitNotifications/MSS_HitMarker_Shield_Destroyed_Crit
inputs [
 Retrigger.GainAdjustment {type=Float, access=Constructor}
 Retrigger.Threshold {type=Float, access=Constructor}
 Source.CrashWaves {type=WaveAsset:Array, access=Constructor}
 Source.EnemyImpactWaves {type=WaveAsset:Array, access=Constructor}
 Source.SynthWaves {type=WaveAsset:Array, access=Constructor}
 Time.ElapsedTime {type=Float}
 Time.PlayTime {type=Float}
 Time.TimeSinceLastPlayed {type=Float}
 UE.Source.OnPlay {type=Trigger}
]
interfaces [Fort.Time:1.0, UE.OutputFormat.Stereo:1.0, UE.Source.OneShot:1.0, UE.Source:1.1]
outputs [
 "UE.OutputFormat.Stereo.Audio:0" {type=Audio}
 "UE.OutputFormat.Stereo.Audio:1" {type=Audio}
 UE.Source.OneShot.OnFinished {type=Trigger}
]
pages [Default {
  input_defaults {
   Retrigger.GainAdjustment=-2
   Retrigger.Threshold=0.75
   Time.ElapsedTime=0
   Time.PlayTime=0
   Time.TimeSinceLastPlayed=0.016
   UE.Source.OnPlay=null
   Source.CrashWaves [&/sfx_hitmarker_notification_critical_layer]
   Source.EnemyImpactWaves [
    /Game/Sounds/Fort_Impact_Sounds/Flesh/player_impact_enemy_stereo_04
    /Game/Sounds/Fort_Impact_Sounds/Flesh/player_impact_enemy_stereo_03
    /Game/Sounds/Fort_Impact_Sounds/Flesh/player_impact_enemy_stereo_02
    /Game/Sounds/Fort_Impact_Sounds/Flesh/player_impact_enemy_stereo_01
   ]
   Source.SynthWaves [&/HitMarker_ShieldBreak_01, &/HitMarker_ShieldBreak_02]
  }
  nodes [
   n1 = TriggerCompare.Clamp:Float(A=^Retrigger.Threshold, B=^Time.TimeSinceLastPlayed) {also {defaults {Type=3}}}
   TriggerRouteTriggerRouteTime2_0 = "TriggerRoute.Trigger Route (Time, 2):Time"(Set 1=@n1.FALSE, Set 0=@n1.TRUE) {also {defaults {Value 0=0.2, Value 1=0.15}}}
   TriggerDelay_1 = UE.Trigger Delay(In=@TriggerRouteTriggerRouteTime2_0.On Set, Delay Time=@TriggerRouteTriggerRouteTime2_0.Value)
   n2 = /Game/Sounds/Fort_Gadgets/Shield/MSS_Shield_Deactivate_Base(
    "Retrigger.GainAdjustment (Db)"=^Retrigger.GainAdjustment
    Retrigger.BehaviorThresholdTime=^Retrigger.Threshold
    Source.Waves=^Source.CrashWaves
    Time.ElapsedTime=^Time.ElapsedTime
    Time.PlayTime=^Time.PlayTime
    Time.TimeSinceLastPlayed=^Time.TimeSinceLastPlayed
    UE.Source.OnPlay=@TriggerDelay_1.out
   ) {also {defaults {Retrigger.FadeInTime=0.15, Retrigger.TimeOffset=0.15}}}
   n3 = &MSS_Shield_Deactivate_Base(
    "Retrigger.GainAdjustment (Db)"=^Retrigger.GainAdjustment
    Retrigger.BehaviorThresholdTime=^Retrigger.Threshold
    Source.Waves=^Source.EnemyImpactWaves
    Time.ElapsedTime=^Time.ElapsedTime
    Time.PlayTime=^Time.PlayTime
    Time.TimeSinceLastPlayed=^Time.TimeSinceLastPlayed
    UE.Source.OnPlay=@TriggerDelay_1.out
   ) {also {defaults {Retrigger.TimeOffset=0.15}}}
   TriggerDelay_0 = UE.Trigger Delay(In=^UE.Source.OnPlay) {also {defaults {Delay Time=0.15}}}
   n4 = &MSS_Shield_Deactivate_Base(
    "Retrigger.GainAdjustment (Db)"=^Retrigger.GainAdjustment
    Retrigger.BehaviorThresholdTime=^Retrigger.Threshold
    Source.Waves=^Source.SynthWaves
    Time.ElapsedTime=^Time.ElapsedTime
    Time.PlayTime=^Time.PlayTime
    Time.TimeSinceLastPlayed=^Time.TimeSinceLastPlayed
    UE.Source.OnPlay=@TriggerDelay_0.out
   ) {also {defaults {Retrigger.TimeOffset=0.2}}}
   AudioMixerAudioMixerStereo3_0 = "AudioMixer.Audio Mixer (Stereo, 3)"(
    In 2 L=@n2."UE.OutputFormat.Stereo.Audio:0"
    In 2 R=@n2."UE.OutputFormat.Stereo.Audio:1"
    In 1 L=@n3."UE.OutputFormat.Stereo.Audio:0"
    In 1 R=@n3."UE.OutputFormat.Stereo.Audio:1"
    In 0 L=@n4."UE.OutputFormat.Stereo.Audio:0"
    In 0 R=@n4."UE.OutputFormat.Stereo.Audio:1"
   ) {also {defaults {Gain 2=1.25}}}
   Multiply_0 = UE.Multiply:Audio by Float(PrimaryOperand=@AudioMixerAudioMixerStereo3_0.Out L) {also {defaults {AdditionalOperands=0.6}}}
   Multiply_1 = UE.Multiply:Audio by Float(PrimaryOperand=@AudioMixerAudioMixerStereo3_0.Out R) {also {defaults {AdditionalOperands=0.6}}}
   TriggerAnyTriggerAny3_0 = "TriggerAny.Trigger Any (3)"(In 2=@n2.UE.Source.OneShot.OnFinished, In 1=@n3.UE.Source.OneShot.OnFinished, In 0=@n4.UE.Source.OneShot.OnFinished)
  ]
  links [
   {dst="@UE.OutputFormat.Stereo.Audio:0", src=Multiply_0.out}
   {dst="@UE.OutputFormat.Stereo.Audio:1", src=Multiply_1.out}
   {dst="@UE.Source.OneShot.OnFinished", src=TriggerAnyTriggerAny3_0.out}
  ]
 }]
properties {
 assetImportData=&self:AssetImportData
 attenuationSettings=&/Fort_HitMarker_Instigate_Att
 numChannels=2
 outputFormat=Stereo
 sampleRate=48000
 soundClassObject=&Fort_2DHitNotifications_Instigate_PSM
 concurrencySet [/Game/Sounds/Concurrency_Rules/HitNotifications_Critical_StopOldest]
}
