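# Parent MetaSound source for legacy hit notification sounds, a stereo one-shot with a concurrency limit.
# It mixes a stereo Layer A and mono Layer B, each with random wave and pitch and optional delay.
# Replays within a time threshold apply alternate gain, start offset and fade-in.
# Random volume and patch gain scale the mix, which pans by hit-marker azimuth with adjustable stereo spread.
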MetaSoundSource /Game/Athena/Sounds/HitNotifications/MSS_HitNotifies_Legacy_Parent
inputs [
 HitMarker.Azimuth {type=Float}
 LayerA.Gain {type=Float}
 LayerA.OptionalDelay {type=Time}
 LayerA.Pitch.Max {type=Float}
 LayerA.Pitch.Min {type=Float}
 LayerA.Wave {type=WaveAsset:Array}
 LayerB.Gain {type=Float}
 LayerB.OptionalDelay {type=Time}
 LayerB.Pitch.Max {type=Float}
 LayerB.Pitch.Min {type=Float}
 LayerB.Wave {type=WaveAsset:Array}
 Patch.Gain {type=Float}
 Patch.StereoSpread {type=Float}
 Patch.Vol.Max {type=Float}
 Patch.Vol.Min {type=Float}
 Retrigger.BehaviorThresholdTime {type=Float}
 Retrigger.FadeInTime {type=Float}
 "Retrigger.GainAdjustment (Db)" {type=Float}
 Retrigger.TimeOffset {type=Time}
 Time.ElapsedTime {type=Float}
 Time.GameTimeSeconds {type=Float}
 Time.PlayTime {type=Float}
 Time.TimeSinceLastPlayed {type=Float}
 UE.Source.OnPlay {type=Trigger}
 UE.Source.Orientation.Azimuth {type=Float}
 UE.Source.Orientation.Elevation {type=Float}
]
interfaces [Fort.Time:1.0, UE.OutputFormat.Stereo:1.0, UE.Source.OneShot:1.0, UE.Source.Orientation:1.0, UE.Source:1.1]
outputs [
 "UE.OutputFormat.Stereo.Audio:0" {type=Audio}
 "UE.OutputFormat.Stereo.Audio:1" {type=Audio}
 UE.Source.OneShot.OnFinished {type=Trigger}
]
pages [Default {
  input_defaults {
   HitMarker.Azimuth=0
   LayerA.Gain=1
   LayerA.OptionalDelay=0
   LayerA.Pitch.Max=0
   LayerA.Pitch.Min=0
   LayerB.Gain=1
   LayerB.OptionalDelay=0
   LayerB.Pitch.Max=1
   LayerB.Pitch.Min=-1
   Patch.Gain=1
   Patch.StereoSpread=0
   Patch.Vol.Max=1
   Patch.Vol.Min=1
   Retrigger.BehaviorThresholdTime=0.75
   Retrigger.FadeInTime=0
   "Retrigger.GainAdjustment (Db)"=0
   Retrigger.TimeOffset=0
   Time.ElapsedTime=0
   Time.GameTimeSeconds=0
   Time.PlayTime=0
   Time.TimeSinceLastPlayed=0
   UE.Source.OnPlay=null
   UE.Source.Orientation.Azimuth=0
   UE.Source.Orientation.Elevation=0
   LayerA.Wave []
   LayerB.Wave []
  }
  nodes [
   n1 = Clamp.Clamp:Float(In=UE.Divide:Float(PrimaryOperand=^HitMarker.Azimuth) {also {defaults {AdditionalOperands=90}}}.Out) {also {defaults {Max=1, Min=-1}}}
   TriggerDelay_0 = UE.Trigger Delay(In=^UE.Source.OnPlay) {also {defaults {Delay Time=0.001}}}
   n2 = TriggerCompare.Clamp:Float(A=^Time.TimeSinceLastPlayed, Compare=@TriggerDelay_0.Out) {also {defaults {B=0}}}
   n3 = TriggerCompare.Clamp:Float(B=^Retrigger.BehaviorThresholdTime, A=^Time.TimeSinceLastPlayed, Compare=@n2.False) {also {defaults {Type=3}}}
   TriggerAnyTriggerAny2_0 = "TriggerAny.Trigger Any (2)"(In 0=@n2.True, In 1=@n3.True)
   n4 = VariableAccessor.Trigger(Variable=VariableMutator.Trigger(Variable=InitVariable.Trigger().Variable, Value=@TriggerAnyTriggerAny2_0.Out).Variable)
   n5 = VariableAccessor.Trigger(Variable=@n4.Variable)
   n6 = VariableAccessor.Trigger(Variable=VariableMutator.Trigger(Variable=InitVariable.Trigger().Variable, Value=@n3.False).Variable)
   n7 = VariableAccessor.Trigger(Variable=@n6.Variable)
   TriggerRouteTriggerRouteFloat2_0 = "TriggerRoute.Trigger Route (Float, 2):Float"(Value 1=^"Retrigger.GainAdjustment (Db)", Set 0=@n5.Value, Set 1=@n7.Value) {also {defaults {Value 0=0}}}
   DecibelstoLinearGain_0 = UE.Decibels to Linear Gain:Float(Decibels=@TriggerRouteTriggerRouteFloat2_0.Value)
   n8 = UE.Multiply:Float(PrimaryOperand=^LayerA.Gain, AdditionalOperands=@DecibelstoLinearGain_0.Linear Gain)
   n9 = UE.Multiply:Float(AdditionalOperands=^LayerB.Gain, PrimaryOperand=@DecibelstoLinearGain_0.Linear Gain)
   TriggerDelay_1 = UE.Trigger Delay(Delay Time=^LayerB.OptionalDelay, In=^UE.Source.OnPlay)
   ArrayRandomGet_0 = Array.Random Get:WaveAsset:Array(In Array=^LayerB.Wave, Next=@TriggerDelay_1.Out)
   n10 = UE.RandomFloat(Max=^LayerB.Pitch.Max, Min=^LayerB.Pitch.Min, Next=@ArrayRandomGet_0.On Next)
   TriggerRouteTriggerRouteTime2_0 = "TriggerRoute.Trigger Route (Time, 2):Time"(Value 1=^Retrigger.TimeOffset, Set 0=@n4.Value, Set 1=@n6.Value) {also {defaults {Value 0=0}}}
   WavePlayer_0 = UE.Wave Player:Mono(
    Wave Asset=@ArrayRandomGet_0.Value
    Play=@n10.On Next
    Pitch Shift=@n10.Value
    Start Time=UE.Reroute(Value=UE.Reroute(Value=@TriggerRouteTriggerRouteTime2_0.Value).Value).Value
   ) {also {version [1,1]}}
   TriggerAccumulatorTriggerAccumulate2_0 = "TriggerAccumulator.Trigger Accumulate (2)"(In 0=@TriggerRouteTriggerRouteFloat2_0.On Set, In 1=@TriggerRouteTriggerRouteTime2_0.On Set)
   TriggerDelay_2 = UE.Trigger Delay(Delay Time=^LayerA.OptionalDelay, In=@TriggerAccumulatorTriggerAccumulate2_0.Out)
   ArrayRandomGet_1 = Array.Random Get:WaveAsset:Array(In Array=^LayerA.Wave, Next=@TriggerDelay_2.Out)
   n11 = UE.RandomFloat(Max=^LayerA.Pitch.Max, Min=^LayerA.Pitch.Min, Next=@ArrayRandomGet_1.On Next)
   WavePlayer_1 = UE.Wave Player:Stereo(
    Wave Asset=@ArrayRandomGet_1.Value
    Play=@n11.On Next
    Pitch Shift=@n11.Value
    Start Time=UE.Reroute(Value=UE.Reroute(Value=@TriggerRouteTriggerRouteTime2_0.Value).Value).Value
   ) {also {version [1,1]}}
   AudioMixerAudioMixerStereo2_0 = "AudioMixer.Audio Mixer (Stereo, 2)"(
    Gain 0=UE.Reroute(Value=UE.Reroute(Value=@n8.Out).Value).Value
    Gain 1=UE.Reroute(Value=UE.Reroute(Value=@n9.Out).Value).Value
    In 1 L=@WavePlayer_0.Out Mono
    In 1 R=@WavePlayer_0.Out Mono
    In 0 L=@WavePlayer_1.Out Left
    In 0 R=@WavePlayer_1.Out Right
   )
   TriggerRouteTriggerRouteFloat2_1 = "TriggerRoute.Trigger Route (Float, 2):Float"(
    Value 1=^Retrigger.FadeInTime
    Set 1=VariableAccessor.Trigger(Variable=@n7.Variable).Value
    Set 0=VariableAccessor.Trigger(Variable=@n5.Variable).Value
   ) {also {defaults {Value 0=0}}}
   n12 = UE.Reroute(Value=@WavePlayer_1.On Play)
   n13 = UE.Multiply:Float(
    AdditionalOperands=&MSP_Fade(In=@TriggerRouteTriggerRouteFloat2_1.On Set, Fade Time=@TriggerRouteTriggerRouteFloat2_1.Value).Value Out
    PrimaryOperand=UE.Multiply:Float(AdditionalOperands=^Patch.Gain, PrimaryOperand=UE.RandomFloat(Max=^Patch.Vol.Max, Min=^Patch.Vol.Min, Next=@n12.Value).Value).Out
   )
   StereoPanner_0 = UE.Stereo Panner(
    Pan Amount=Clamp.Clamp:Float(In=UE.Add:Float(AdditionalOperands=^Patch.StereoSpread, PrimaryOperand=@n1.Value).Out) {also {defaults {Max=1, Min=-1}}}.Value
    In=UE.Reroute(Value=UE.Multiply:Audio by Float(PrimaryOperand=@AudioMixerAudioMixerStereo2_0.Out R, AdditionalOperands=@n13.Out).Out).Value
   )
   StereoPanner_1 = UE.Stereo Panner(
    Pan Amount=Clamp.Clamp:Float(In=UE.Subtract:Float(AdditionalOperands=^Patch.StereoSpread, PrimaryOperand=@n1.Value).Out) {also {defaults {Max=1, Min=-1}}}.Value
    In=UE.Multiply:Audio by Float(PrimaryOperand=@AudioMixerAudioMixerStereo2_0.Out L, AdditionalOperands=@n13.Out).Out
   )
   Add_0 = UE.Add:Audio(AdditionalOperands=@StereoPanner_0.Out Right, PrimaryOperand=@StereoPanner_1.Out Right)
   Add_1 = UE.Add:Audio(AdditionalOperands=@StereoPanner_0.Out Left, PrimaryOperand=@StereoPanner_1.Out Left)
   TriggerAccumulatorTriggerAccumulate2_1 = "TriggerAccumulator.Trigger Accumulate (2)"(In 1=@WavePlayer_0.On Finished, In 0=@WavePlayer_1.On Finished)
  ]
  links [
   {dst="@UE.OutputFormat.Stereo.Audio:1", src=Add_0.Out}
   {dst="@UE.OutputFormat.Stereo.Audio:0", src=Add_1.Out}
   {dst="@UE.Source.OneShot.OnFinished", src=TriggerAccumulatorTriggerAccumulate2_1.Out}
  ]
 }]
properties {
 assetImportData=&self:AssetImportData
 numChannels=2
 outputFormat=Stereo
 sampleRate=48000
 concurrencySet [&HitNotifications_TwoMaximum_StopOldest_VolumeScale]
}
