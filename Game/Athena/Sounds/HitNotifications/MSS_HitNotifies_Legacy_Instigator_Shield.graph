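# Stereo one-shot MetaSound preset for the legacy shield hit-marker notification heard by the instigator of a hit.
# It inherits its graph from MSS_HitNotifies_Legacy_Parent, which exposes two wave layers plus gain, stereo spread, azimuth, time and retrigger inputs.
# Layer A has no waves, while Layer B supplies four shield hit-marker variations with a pitch range.
# It routes to the 2D instigator hit-notification sound class, with two-voice stop-oldest concurrency.
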MetaSoundSource /Game/Athena/Sounds/HitNotifications/MSS_HitNotifies_Legacy_Instigator_Shield
preset_parent=&/MSS_HitNotifies_Legacy_Parent
inputs [
 HitMarker.Azimuth {type=Float}
 LayerA.Gain {type=Float}
 LayerA.OptionalDelay {type=Time}
 LayerA.Pitch.Max {type=Float}
 LayerA.Pitch.Min {type=Float}
 LayerA.Wave {type=WaveAsset:Array}
 LayerB.Gain {type=Float}
 LayerB.OptionalDelay {type=Time}
 LayerB.Pitch.Max {type=Float}
 LayerB.Pitch.Min {type=Float}
 LayerB.Wave {type=WaveAsset:Array}
 Patch.Gain {type=Float}
 Patch.StereoSpread {type=Float}
 Patch.Vol.Max {type=Float}
 Patch.Vol.Min {type=Float}
 Retrigger.BehaviorThresholdTime {type=Float}
 Retrigger.FadeInTime {type=Float}
 "Retrigger.GainAdjustment (Db)" {type=Float}
 Retrigger.TimeOffset {type=Time}
 Time.ElapsedTime {type=Float}
 Time.GameTimeSeconds {type=Float}
 Time.PlayTime {type=Float}
 Time.TimeSinceLastPlayed {type=Float}
 UE.Source.OnPlay {type=Trigger}
 UE.Source.Orientation.Azimuth {type=Float}
 UE.Source.Orientation.Elevation {type=Float}
]
interfaces [Fort.Time:1.0, UE.OutputFormat.Stereo:1.0, UE.Source.OneShot:1.0, UE.Source.Orientation:1.0, UE.Source:1.1]
outputs [
 "UE.OutputFormat.Stereo.Audio:0" {type=Audio}
 "UE.OutputFormat.Stereo.Audio:1" {type=Audio}
 UE.Source.OneShot.OnFinished {type=Trigger}
]
pages [Default {input_defaults {
   HitMarker.Azimuth=0
   LayerA.Gain=1
   LayerA.OptionalDelay=0
   LayerA.Pitch.Max=0
   LayerA.Pitch.Min=0
   LayerB.Gain=1
   LayerB.OptionalDelay=0
   LayerB.Pitch.Max=1
   LayerB.Pitch.Min=-1
   Patch.Gain=0.6
   Patch.StereoSpread=0
   Patch.Vol.Max=1
   Patch.Vol.Min=1
   Retrigger.BehaviorThresholdTime=0.75
   Retrigger.FadeInTime=0
   "Retrigger.GainAdjustment (Db)"=0
   Retrigger.TimeOffset=0
   Time.ElapsedTime=0
   Time.GameTimeSeconds=0
   Time.PlayTime=0
   Time.TimeSinceLastPlayed=0
   UE.Source.OnPlay=null
   UE.Source.Orientation.Azimuth=0
   UE.Source.Orientation.Elevation=0
   LayerA.Wave []
   LayerB.Wave [
    &/HitMarker_Shield_3P_01
    &/HitMarker_Shield_3P_02
    &/HitMarker_Shield_3P_03
    &/HitMarker_Shield_3P_04
   ]
  }}]
properties {
 assetImportData=&self:AssetImportData
 attenuationSettings=&/Att_HitMarker_Instigate_MS
 numChannels=2
 outputFormat=Stereo
 sampleRate=48000
 soundClassObject=&Fort_2DHitNotifications_Instigate_PSM
 volume=0.5
 concurrencySet [&HitNotifications_TwoMaximum_StopOldest_VolumeScale]
}
