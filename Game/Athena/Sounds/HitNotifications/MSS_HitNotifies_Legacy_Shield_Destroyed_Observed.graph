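# Stereo one-shot MetaSound preset for the legacy shield-destroyed hit notification, observed variant.
# The graph comes from MSS_HitNotifies_Legacy_Parent; this preset supplies a 3D shield-break wave to Layer A and leaves Layer B empty.
# Exposed inputs set per-layer pitch ranges, patch gain and retrigger behavior with threshold, fade-in, time offset and gain cut.
# Observed hit-marker attenuation spatializes it, and concurrency settings limit it to two voices.
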MetaSoundSource /Game/Athena/Sounds/HitNotifications/MSS_HitNotifies_Legacy_Shield_Destroyed_Observed
preset_parent=&/MSS_HitNotifies_Legacy_Parent
inputs [
 LayerA.Gain {type=Float}
 LayerA.OptionalDelay {type=Time}
 LayerA.Pitch.Max {type=Float}
 LayerA.Pitch.Min {type=Float}
 LayerA.Wave {type=WaveAsset:Array}
 LayerB.Gain {type=Float}
 LayerB.OptionalDelay {type=Time}
 LayerB.Pitch.Max {type=Float}
 LayerB.Pitch.Min {type=Float}
 LayerB.Wave {type=WaveAsset:Array}
 Patch.Gain {type=Float}
 Patch.StereoSpread {type=Float}
 Patch.Vol.Max {type=Float}
 Patch.Vol.Min {type=Float}
 Retrigger.BehaviorThresholdTime {type=Float, access=Constructor}
 Retrigger.FadeInTime {type=Float, access=Constructor}
 "Retrigger.GainAdjustment (Db)" {type=Float, access=Constructor}
 Retrigger.TimeOffset {type=Time, access=Constructor}
 Time.ElapsedTime {type=Float}
 Time.GameTimeSeconds {type=Float}
 Time.PlayTime {type=Float}
 Time.TimeSinceLastPlayed {type=Float}
 UE.Source.OnPlay {type=Trigger}
 UE.Source.Orientation.Azimuth {type=Float}
 UE.Source.Orientation.Elevation {type=Float}
]
interfaces [Fort.Time:1.0, UE.OutputFormat.Stereo:1.0, UE.Source.OneShot:1.0, UE.Source.Orientation:1.0, UE.Source:1.1]
outputs [
 "UE.OutputFormat.Stereo.Audio:0" {type=Audio}
 "UE.OutputFormat.Stereo.Audio:1" {type=Audio}
 UE.Source.OneShot.OnFinished {type=Trigger}
]
pages [Default {input_defaults {
   LayerA.Gain=1
   LayerA.OptionalDelay=0
   LayerA.Pitch.Max=1
   LayerA.Pitch.Min=-1
   LayerB.Gain=1
   LayerB.OptionalDelay=0
   LayerB.Pitch.Max=1
   LayerB.Pitch.Min=-1
   Patch.Gain=0.5
   Patch.StereoSpread=0
   Patch.Vol.Max=1
   Patch.Vol.Min=1
   Retrigger.BehaviorThresholdTime=0.75
   Retrigger.FadeInTime=0.1
   "Retrigger.GainAdjustment (Db)"=-4
   Retrigger.TimeOffset=0.15
   Time.ElapsedTime=0
   Time.GameTimeSeconds=0
   Time.PlayTime=0
   Time.TimeSinceLastPlayed=0
   UE.Source.OnPlay=null
   UE.Source.Orientation.Azimuth=0
   UE.Source.Orientation.Elevation=0
   LayerA.Wave [&/HitMarker_ShieldBreak_3D_01]
   LayerB.Wave []
  }}]
properties {
 assetImportData=&self:AssetImportData
 attenuationSettings=&/Fort_HitMarker_Observed_Att
 bOverrideConcurrency=true
 maxDistance=5500
 numChannels=2
 outputFormat=Stereo
 sampleRate=48000
 soundClassObject=/Game/Sounds/FortSoundClassesAndModes/Fort_Impact_Player_PSM
 concurrencyOverrides {maxCount=2, resolutionRule=StopLowestPriority}
 concurrencySet [&HitNotifications_TwoMaximum_StopOldest_VolumeScale]
}
