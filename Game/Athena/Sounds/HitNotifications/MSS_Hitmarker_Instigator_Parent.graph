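# Parent MetaSound Source for instigator hitmarker sounds: a stereo one-shot hit notification, defaulting to three HitMarker_Body waves.
# On play it picks a random wave from the Sounds array, then rolls a random pitch shift before starting the wave player.
# An azimuth panner places the audio by source orientation and stereo spread, and a separately randomized volume scales both channels.
# Exposed pitch and volume ranges default to no variation; the wave's end fires OnFinished.
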MetaSoundSource /Game/Athena/Sounds/HitNotifications/MSS_Hitmarker_Instigator_Parent
inputs [
 Max Pitch {type=Float}
 Max Random Volume {type=Float}
 Min Pitch {type=Float}
 Min Random Volume {type=Float}
 Sounds {type=WaveAsset:Array}
 StereoSpread {type=Float}
 UE.Source.OnPlay {type=Trigger}
 UE.Source.Orientation.Azimuth {type=Float}
 UE.Source.Orientation.Elevation {type=Float}
]
interfaces [UE.OutputFormat.Stereo:1.0, UE.Source.OneShot:1.0, UE.Source.Orientation:1.0, UE.Source:1.1]
outputs [
 "UE.OutputFormat.Stereo.Audio:0" {type=Audio}
 "UE.OutputFormat.Stereo.Audio:1" {type=Audio}
 UE.Source.OneShot.OnFinished {type=Trigger}
]
pages [Default {
  input_defaults {
   Max Pitch=1
   Max Random Volume=1
   Min Pitch=1
   Min Random Volume=1
   StereoSpread=0
   UE.Source.OnPlay=null
   UE.Source.Orientation.Azimuth=0
   UE.Source.Orientation.Elevation=0
   Sounds [
    &/HitMarker_Body_01
    &/HitMarker_Body_02
    &/HitMarker_Body_03
   ]
  }
  nodes [
   ArrayRandomGet_0 = Array.Random Get:WaveAsset:Array(In Array=^Sounds, Next=^UE.Source.OnPlay)
   n1 = UE.RandomFloat(Max=^Max Pitch, Min=^Min Pitch, Next=@ArrayRandomGet_0.On Next)
   WavePlayer_0 = UE.Wave Player:Stereo(Play=@n1.On Next, Pitch Shift=@n1.Value, Wave Asset=UE.Reroute(Value=UE.Reroute(Value=@ArrayRandomGet_0.Value).Value).Value)
   n2 = &AzimuthPannerStereo(StereoSpread=^StereoSpread, Azimuth=^UE.Source.Orientation.Azimuth, In L=@WavePlayer_0.Out Left, In R=@WavePlayer_0.Out Right)
   n3 = UE.RandomFloat(Max=^Max Random Volume, Min=^Min Random Volume, Next=^UE.Source.OnPlay)
   Multiply_0 = UE.Multiply:Audio by Float(PrimaryOperand=@n2.Out L, AdditionalOperands=@n3.Value)
   Multiply_1 = UE.Multiply:Audio by Float(PrimaryOperand=@n2.Out R, AdditionalOperands=@n3.Value)
  ]
  links [
   {dst="@UE.OutputFormat.Stereo.Audio:0", src=Multiply_0.Out}
   {dst="@UE.OutputFormat.Stereo.Audio:1", src=Multiply_1.Out}
   {dst="@UE.Source.OneShot.OnFinished", src=WavePlayer_0.On Finished}
  ]
 }]
properties {
 assetImportData=&self:AssetImportData
 attenuationSettings=&/Att_HitMarker_Instigate_MS
 numChannels=2
 outputFormat=Stereo
 sampleRate=48000
 soundClassObject=&Fort_2DHitNotifications_Instigate_PSM
 concurrencySet [&HitNotifications_TwoMaximum_StopOldest_VolumeScale]
}
