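# MetaSound Source preset that supplies the stereo one-shot shield hitmarker sound for the instigator of a hit.
# It inherits its graph from MSS_Hitmarker_Instigator_Parent and feeds it four shield hitmarker waves, with pitch and volume ranges flat and stereo spread at zero.
# Playback uses hitmarker attenuation, a 2D hit-notification sound class, and concurrency named for two voices, stop-oldest and volume scaling.
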MetaSoundSource /Game/Athena/Sounds/HitNotifications/MSS_Hitmarker_Instigator_Shield
preset_parent=&/MSS_Hitmarker_Instigator_Parent
inputs [
 Max Pitch {type=Float}
 Max Random Volume {type=Float}
 Min Pitch {type=Float}
 Min Random Volume {type=Float}
 Sounds {type=WaveAsset:Array}
 StereoSpread {type=Float}
 UE.Source.OnPlay {type=Trigger}
 UE.Source.Orientation.Azimuth {type=Float}
 UE.Source.Orientation.Elevation {type=Float}
]
interfaces [UE.OutputFormat.Stereo:1.0, UE.Source.OneShot:1.0, UE.Source.Orientation:1.0, UE.Source:1.1]
outputs [
 "UE.OutputFormat.Stereo.Audio:0" {type=Audio}
 "UE.OutputFormat.Stereo.Audio:1" {type=Audio}
 UE.Source.OneShot.OnFinished {type=Trigger}
]
pages [Default {input_defaults {
   Max Pitch=1
   Max Random Volume=1
   Min Pitch=1
   Min Random Volume=1
   StereoSpread=0
   UE.Source.OnPlay=null
   UE.Source.Orientation.Azimuth=0
   UE.Source.Orientation.Elevation=0
   Sounds [
    &/HitMarker_Shield_3P_01
    &/HitMarker_Shield_3P_02
    &/HitMarker_Shield_3P_03
    &/HitMarker_Shield_3P_04
   ]
  }}]
properties {
 assetImportData=&self:AssetImportData
 attenuationSettings=&/Att_HitMarker_Instigate_MS
 numChannels=2
 outputFormat=Stereo
 sampleRate=48000
 soundClassObject=&Fort_2DHitNotifications_Instigate_PSM
 volume=0.6
 concurrencySet [&HitNotifications_TwoMaximum_StopOldest_VolumeScale]
}
