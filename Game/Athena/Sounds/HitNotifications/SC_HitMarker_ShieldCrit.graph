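# Sound cue for the critical shield hit marker notification, judging by its name and the shield and critical-layer waves it plays.
# A mixer layers a randomly chosen one of four shield hit waves, through a modulator, with a slightly louder, briefly delayed critical notification layer.
# The cue routes to a 2D hit notification sound class with instigator attenuation and a two-voice, stop-oldest concurrency limit.
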SoundCue /Game/Athena/Sounds/HitNotifications/SC_HitMarker_ShieldCrit
root=Mixer_1
nodes [
 {id=Delay_6, children [WavePlayer_4], properties {delayMax=0.1, delayMin=0.1}}
 {id=Mixer_1, children [Modulator_0,Modulator_6], properties {inputVolume [1,1.2]}}
 {id=Modulator_0, children [Random_0], properties {volumeMax=1, volumeMin=1}}
 {id=Modulator_6, children [Delay_6], properties {pitchMax=1, pitchMin=1}}
 {
  id=Random_0
  children [WavePlayer_0,WavePlayer_1,WavePlayer_2,WavePlayer_3]
  properties {bShouldExcludeFromBranchCulling=true, weights [1,1,1,1]}
 }
 {id=WavePlayer_0, properties {soundWaveAssetPtr=&/HitMarker_Shield_3P_01}}
 {id=WavePlayer_1, properties {soundWaveAssetPtr=&/HitMarker_Shield_3P_02}}
 {id=WavePlayer_2, properties {soundWaveAssetPtr=&/HitMarker_Shield_3P_03}}
 {id=WavePlayer_3, properties {soundWaveAssetPtr=&/HitMarker_Shield_3P_04}}
 {id=WavePlayer_4, properties {soundWaveAssetPtr=&/sfx_hitmarker_notification_critical_layer}}
]
properties {
 attenuationSettings=&/Fort_HitMarker_Instigate_Att
 maxDistance=2097152
 soundClassObject=&Fort_2DHitNotifications_Instigate_PSM
 volumeMultiplier=0.6
 concurrencyOverrides {maxCount=2, resolutionRule=StopOldest}
 concurrencySet [&HitNotifications_TwoMaximum_StopOldest_VolumeScale]
}
